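# Material function that selects a Tibanna gas color gradient, sampled at CurveTime, for weapon effects.
# Eight curve atlas row parameters from the TibannaGasColors atlas supply red, orange, yellow, green, cyan, blue, purple and dark gradients.
# Curve Select (0-1) is scaled by seven and stepped down through subtractions into clamped alphas for chained lerps.
# The chain works as a branchless selector that blends smoothly between neighboring gradients.
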MaterialFunction /Game/Effects/Fort_Effects/CUrves/Weapons/TibannaGasColorSelection
graph {nodes [
  n1 = FunctionInput() {inputName=CurveTime, inputType=FunctionInput_Scalar}
  n2 = CurveAtlasRowParameter(CurveTime=@n1) {
   atlas=&/TibannaGasColors
   curve=&/Curve_Red
   parameterName=Tibanna Gas Color_RED
  }
  n3 = CurveAtlasRowParameter(CurveTime=@n1) {
   atlas=&TibannaGasColors
   curve=&/Curve_Orange
   defaultValue=1
   parameterName=Tibanna Gas Color_ORANGE
  }
  n4 = Multiply(A=FunctionInput() {description=0-1 Value, inputName=Curve Select, inputType=FunctionInput_Scalar, sortPriority=1}) {constB=7}
  n5 = CurveAtlasRowParameter(CurveTime=@n1) {
   atlas=&TibannaGasColors
   curve=&/Curve_Yellow
   defaultValue=2
   parameterName=Tibanna Gas Color_YELLOW
  }
  n6 = Subtract(A=@n4)
  n7 = LinearInterpolate(A=LinearInterpolate(A=@n2, B=@n3, Alpha=Clamp(None=@n4)), B=@n5, Alpha=Clamp(None=@n6))
  n8 = CurveAtlasRowParameter(CurveTime=@n1) {
   atlas=&TibannaGasColors
   curve=&/Curve_Green
   defaultValue=3
   parameterName=Tibanna Gas Color_GREEN
  }
  n9 = Subtract(A=@n6)
  n10 = Clamp(None=@n9)
  n11 = CurveAtlasRowParameter(CurveTime=@n1) {
   atlas=&TibannaGasColors
   curve=&/Curve_Cyan
   defaultValue=4
   parameterName=Tibanna Gas Color_CYAN
  }
  n12 = Subtract(A=@n9)
  n13 = CurveAtlasRowParameter(CurveTime=@n1) {
   atlas=&TibannaGasColors
   curve=&/Curve_Blue
   defaultValue=5
   parameterName=Tibanna Gas Color_BLUE
  }
  n14 = Subtract(A=@n12)
  n15 = LinearInterpolate(A=LinearInterpolate(A=LinearInterpolate(A=@n7, B=@n8, Alpha=@n10), B=@n11, Alpha=Clamp(None=@n12)), B=@n13, Alpha=Clamp(None=@n14))
  n16 = CurveAtlasRowParameter(CurveTime=@n1) {
   atlas=&TibannaGasColors
   curve=&/Curve_Purple
   defaultValue=6
   parameterName=Tibanna Gas Color_PURPLE
  }
  n17 = Subtract(A=@n14)
  n18 = Clamp(None=@n17)
  n19 = CurveAtlasRowParameter(CurveTime=@n1) {
   atlas=&TibannaGasColors
   curve=&/Curve_Dark
   defaultValue=7
   parameterName=Tibanna Gas Color_DARK
  }
  n20 = Subtract(A=@n17)
  FunctionOutput(None=LinearInterpolate(A=LinearInterpolate(A=@n15, B=@n16, Alpha=@n18), B=@n19, Alpha=Clamp(None=@n20)))
 ]}
