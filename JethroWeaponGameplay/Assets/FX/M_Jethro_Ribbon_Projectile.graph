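# Translucent, unlit, two-sided material for Jethro weapon projectile effects on Niagara ribbons and sprites.
# A procedural capsule mask uses UVs and particle size to round the shape, with width varying along its length.
# Color comes from a length ramp through the Tibanna gas curve selection, tinted by particle color and HDR-boosted via dynamic parameters.
# Opacity combines a powered, squeezed shape mask with particle alpha and near-camera fade.
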Material /JethroWeaponGameplay/Assets/FX/M_Jethro_Ribbon_Projectile
graph {
 properties {
  bAutomaticallySetUsageInEditor=false
  bUsedWithNiagaraRibbons=true
  bUsedWithNiagaraSprites=true
  blendMode=BLEND_Translucent
  shadingModel=MSM_Unlit
  twoSided=true
  usedShadingModels=Unlit
 }
 nodes [
  n1 = ComponentMask(None=TextureCoordinate()) {r=true}
  n2 = LinearInterpolate(
   B=Power(Base=NamedRerouteUsage(), Exp=ScalarParameter() {defaultValue=2, parameterName=HDR Power}) {constExponent=0.75}
   Alpha=Power(Base=OneMinus(None=@n1), Exp=ScalarParameter() {defaultValue=0.5544, parameterName=ColorRampExp})
  )
  n3 = DynamicParameter() {defaultValue=<rgba 0 1 1 1>, paramNames [Curve Selection, HDR Intensity, Param3, Param4]}
  n4 = ParticleColor()
  n5 = Multiply(
   A=LinearInterpolate(
    A=@n2
    B=MaterialFunctionCall!(CurveTime=@n2, Curve Select=@n3.Curve Selection) {materialFunction=/Game/Effects/Fort_Effects/CUrves/Weapons/TibannaGasColorSelection}.Result
    Alpha=ScalarParameter() {defaultValue=1, parameterName=Tibanna Gas Color Amount}
   ) {constA=1}
   B=@n4.RGB
  )
  MaterialFunctionCall(Emissive=MaterialFunctionCall(LDR Input=Saturate(None=@n5), HDR Tint=@n5, HDR Intensity Multiplier=@n3.HDR Intensity) {materialFunction=&DeriveHDRfromLDR}.RemappedRange) {materialFunction=&MF_FX_EmissiveAdjustments}
  n6 = ScalarParameter() {defaultValue=1.75, parameterName=Exposure}
  n7 = OneMinus(None=ComponentMask(None=TextureCoordinate()) {r=true})
  n8 = ScalarParameter() {defaultValue=5, parameterName=Squeez_Top_Power}
  n9 = Multiply(A=@n7, B=OneMinus(None=Power(Base=@n7, Exp=@n8) {constExponent=5}))
  n10 = Multiply(
   A=NamedRerouteUsage()
   B=NamedRerouteDeclaration(None=LinearInterpolate(B=@n6, Alpha=@n9) {constB=1.5}) {name=SqueezerTop, nodeColor=<rgba 0.5 0 1 1>}
  )
  n11 = ScalarParameter() {defaultValue=5, parameterName=Opacity Power}
  Multiply(
   A=Multiply(A=@n4.A, B=Power(Base=@n10, Exp=@n11) {constExponent=0.75})
   B=MaterialFunctionCall(Distance=ScalarParameter() {defaultValue=512, parameterName="Distance (S)"}) {materialFunction=&Near_Camera_Fade}.Result
  )
  n12 = TextureCoordinate()
  n13 = Divide(B=ComponentMask(None=@n12) {r=true})
  n14 = ComponentMask(None=TextureCoordinate() {coordinateIndex=1}) {r=true}
  n15 = Min(A=Subtract(A=Multiply(A=@n13, B=@n14), B=@n14), B=@n14)
  n16 = ParticleSize()
  n17 = Multiply(A=Divide(A=Multiply(A=@n15) {constB=100}, B=ComponentMask(None=@n16) {r=true})) {constB=2}
  n18 = ScalarParameter() {defaultValue=2.75, parameterName=Top Bottom Capsule Size}
  n19 = TextureCoordinate()
  n20 = LinearInterpolate(
   A=@n18
   B=Multiply(A=@n18, B=ScalarParameter() {defaultValue=1.25, parameterName=Top Bottom Capsule Size Scale})
   Alpha=Power(Base=ComponentMask(None=@n19) {r=true}, Exp=ScalarParameter() {defaultValue=2, parameterName=Top Bottom Capsule Size Exponent})
  )
  n21 = Multiply(A=ComponentMask(None=@n12) {g=true}, B=@n20) {constB=2}
  n22 = Multiply(A=@n20) {constB=0.5}
  n23 = AppendVector(A=OneMinus(None=Saturate(None=@n17)), B=Abs(None=Subtract(A=@n21, B=@n22)))
  NamedRerouteDeclaration(None=OneMinus(None=Length(None=@n23) {is=class MaterialExpressionLength})) {name=Capsule Shape, nodeColor=<rgba 0.5234 1 0 1>}
 ]
}
