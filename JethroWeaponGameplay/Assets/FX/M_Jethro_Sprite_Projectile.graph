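# Translucent unlit Niagara sprite and ribbon material, named for Jethro weapon projectiles, that draws a soft, glowing radial spot.
# A UV-centered sphere mask, shaped by a contrast power, blends toward a Tibanna gas color curve lookup and is tinted by particle color.
# That color becomes HDR emissive, with Dynamic Parameters driving curve selection and intensity.
# The same mask, raised to an opacity power, multiplies particle alpha for soft edges.
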Material /JethroWeaponGameplay/Assets/FX/M_Jethro_Sprite_Projectile
graph {
 properties {
  bAutomaticallySetUsageInEditor=false
  bUsedWithNiagaraRibbons=true
  bUsedWithNiagaraSprites=true
  blendMode=BLEND_Translucent
  shadingModel=MSM_Unlit
  usedShadingModels=Unlit
 }
 nodes [
  n1 = Constant() {r=0.5}
  n2 = SphereMask(A=TextureCoordinate(), B=@n1, Radius=@n1) {hardnessPercent=0}
  n3 = Power(Base=@n2, Exp=ScalarParameter() {defaultValue=1, parameterName=HDR Color Contrast})
  n4 = DynamicParameter() {defaultValue=<rgba 0 1 1 1>, paramNames [Curve Select, HDR Intensity, Param3, Param4]}
  n5 = ParticleColor()
  n6 = Multiply(
   A=LinearInterpolate(
    A=@n3
    B=MaterialFunctionCall!(CurveTime=@n3, Curve Select=@n4.Curve Select) {materialFunction=/Game/Effects/Fort_Effects/CUrves/Weapons/TibannaGasColorSelection}.Result
    Alpha=ScalarParameter() {defaultValue=1, parameterName=Tibanna Gas Color Amount}
   ) {constA=1}
   B=@n5.RGB
  )
  MaterialFunctionCall(Emissive=MaterialFunctionCall(LDR Input=Saturate(None=@n6), HDR Tint=@n6, HDR Intensity Multiplier=@n4.HDR Intensity) {materialFunction=&DeriveHDRfromLDR}.RemappedRange) {materialFunction=&MF_FX_EmissiveAdjustments}
  Multiply(A=@n5.A, B=Power(Base=@n2, Exp=ScalarParameter() {defaultValue=1, parameterName=Opacity Power}) {constExponent=0.75})
 ]
}
