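# Additive unlit Niagara ribbon material for Jethro weapon tracer smoke trails.
# Dynamic parameters scale, offset and pan the UVs along U, and a tiled swirl normal map distorts them.
# A selected smoke texture channel multiplies particle color and, raised to an erode power, particle alpha.
# The alpha mask then gets near-camera fade, plus depth fade except on the mobile shading path.
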Material /JethroWeaponGameplay/Assets/FX/M_Jethro_Tracer_Smoke
graph {
 properties {
  bAutomaticallySetUsageInEditor=false
  bGenerateSphericalParticleNormals=true
  bUsedWithNiagaraRibbons=true
  blendMode=BLEND_Additive
  shadingModel=MSM_Unlit
  translucencyLightingMode=TLM_VolumetricPerVertexDirectional
  usedShadingModels=Unlit
 }
 nodes [
  n1 = ParticleColor()
  n2 = DynamicParameter() {defaultValue=<rgba 1 1 0 1>, paramNames [Erode Power, PannerSpeed, UV Offset, UV Scale]}
  n3 = Constant() {r=1}
  n4 = Add(A=Multiply(A=TextureCoordinate(), B=AppendVector(A=@n2.UV Scale, B=@n3)), B=AppendVector(A=@n2.UV Offset, B=Constant()))
  n5 = Multiply(A=@n2.PannerSpeed, B=Time())
  n6 = VectorParameter() {parameterName=Distortl SubUVTiling, defaultValue=<rgba 8 8 0 0>}
  n7 = TextureSampleParameter2D(UVs=Multiply(A=TextureCoordinate(), B=ComponentMask(None=@n6.RGB) {g=true, r=true})) {parameterName=DistortTexture, samplerType=SAMPLERTYPE_Normal, texture=/Game/VisualThreatManager/StormVisuals/LargeClouds/T_StormSwirlNormal}
  n8 = ComponentMask(None=@n7.RGB) {g=true, r=true}
  n9 = ScalarParameter() {defaultValue=0.005, group=Shape Control, parameterName=UV Distort}
  n10 = NamedRerouteDeclaration(None=Add(A=Panner(Coordinate=@n4, Time=@n5) {speedX=-1}, B=Multiply(A=@n8, B=@n9))) {name=UV Distortion, nodeColor=<rgba 0 1 0.6953 1>}
  n11 = Saturate(None=DotProduct(
    A=TextureSample(UVs=@n10) {samplerType=SAMPLERTYPE_Grayscale, texture=&T_Smoke_Trail_Soft}.RGB
    B=VectorParameter() {parameterName=ChannelSelect, defaultValue=<rgba 0 1 0 0>}.RGB
   ))
  n12 = Multiply(
   A=@n1.A
   B=Power(Base=@n11, Exp=DynamicParameter() {defaultValue=<rgba 1 1 0 1>, paramNames [Erode Power, PannerSpeed, UV Offset, UV Scale]}.Erode Power)
  )
  Multiply(
   A=MaterialFunctionCall(Distance=ScalarParameter() {defaultValue=150, parameterName=Near Camera Fade Distance}) {materialFunction=&Near_Camera_Fade}.Result
   B=ShadingPathSwitch(Default=DepthFade(Opacity=@n12), Mobile=@n12)
  )
  Multiply(A=@n11, B=@n1.RGB)
 ]
}
