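# Additive unlit material for Niagara mesh particles, by its name the overheat effect on the Jethro assault weapon.
# Particle color feeds emissive through a shared adjustment function, while particle alpha scales a mask built from Fresnel and U-axis gradients.
# A steep U gradient adds to faint panning caustics, whose UVs are warped by a scrolling distortion normal map.
# Fresnel exponent can blend to a dynamic parameter, and mobile shading replaces Fresnel with a flat opacity.
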Material /JethroWeaponGameplay/Assets/FX/Meshes/M_Jethro_Assault_Overheat
graph {
 properties {
  bUsedWithNiagaraMeshParticles=true
  blendMode=BLEND_Additive
  pixelDepthOffsetMode=PDOM_Legacy
  shadingModel=MSM_Unlit
  usedShadingModels=Unlit
 }
 nodes [
  n1 = ParticleColor()
  MaterialFunctionCall(Emissive=@n1.RGB) {materialFunction=&MF_FX_EmissiveAdjustments}
  n2 = Fresnel(
   ExponentIn=LinearInterpolate(
    A=ScalarParameter() {defaultValue=1.4, parameterName=Fresnel}
    B=DynamicParameter() {defaultValue=<rgba 1.4 1 1 1>, paramNames [Fresnel,Param2,Param3,Param4]}.Fresnel
    Alpha=ScalarParameter() {parameterName=Use Dynamic Fresnel}
   )
   BaseReflectFractionIn=Constant()
  )
  n3 = ScalarParameter() {defaultValue=0.45, parameterName=Mobile/Perf Mode Opacity}
  n4 = MaterialFunctionCall(Float2=TextureCoordinate()) {materialFunction=&BreakOutFloat2Components}
  n5 = Power(Base=@n4.R) {constExponent=8.4}
  n6 = TextureCoordinate() {uTiling=0.4, vTiling=0.55}
  n7 = ComponentMask(None=TextureSample(UVs=Panner(Coordinate=@n6) {speedX=0.22}) {samplerType=SAMPLERTYPE_Normal, texture=&T_Weap_Distort_N_02}.RGB) {g=true, r=true}
  n8 = Panner(Coordinate=Add(A=TextureCoordinate() {vTiling=0.18}, B=Multiply(A=@n7) {constB=0.11})) {speedX=0.5}
  n9 = Multiply(
   A=Power(Base=@n4.R) {constExponent=3}
   B=Multiply(A=TextureSample(UVs=@n8) {texture=&T_ShieldCaustics}.R) {constB=0.012}
  )
  Multiply(A=@n1.A, B=Multiply(A=ShadingPathSwitch(Default=@n2, Mobile=@n3) {desc=Migrated from FeatureLevelSwitch}, B=Add(A=@n5, B=@n9)))
 ]
}
