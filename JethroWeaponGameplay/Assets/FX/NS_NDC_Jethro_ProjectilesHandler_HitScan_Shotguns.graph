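# Niagara system that draws hit-scan shotgun tracers from shots read through a Data Channel.
# Active emitters split by instigator filter (weapon owner or others) and NearProjectile flag, with "Up Close" variants using a short fixed lifetime; Tracer_Tracer is disabled.
# Each spawns two-particle ribbons, using RibbonLinkOrder to place head and trailing end along the shot, with curve-shaped bolt length.
# VariantIndex drives dynamic material parameters, and color fades over normalized travel.
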NiagaraSystem /JethroWeaponGameplay/Assets/FX/NS_NDC_Jethro_ProjectilesHandler_HitScan_Shotguns
dependencies {
 usedDataInterfaces [
  &NiagaraDataInterfaceArrayColor
  &NiagaraDataInterfaceCurve
  &NiagaraDataInterfaceArrayFloat
  /Script/Niagara.NiagaraDataInterfaceDataChannelRead
 ]
 usedDynamicInputs []
 usedModules [
  &SystemState
  /Niagara/Modules/System/CompleteIfUnused
  &self:SystemSpawnScript.SystemScriptSource.SystemScriptGraph.NiagaraNodeAssignment_0.SetVariables_#1
 ]
}
emitters {
 Tracer_Ribbons {
  emitterSpawnScript [/SetVariables(Emitter.BoltLength=1500, Emitter.Projectile Velocity=37500) {isSetParameters=true, module=SetVariables_#2}]
  emitterUpdateScript [
   &EmitterState()
   &self:Spawn_NDC_SW(Instigator Filter=2, PositionType=2, Reader:NiagaraDataInterfaceDataChannelRead=$System.NDCReader, Spawn Count=2) {module=Spawn_NDC_SW}
  ]
  particleSpawnScript [
   &self:InitializeParticle_NDC_SW(
    Player Position:NiagaraPosition=&MakeVectorFromLinearColorRGB(Color=$NPC.FortniteNPC.PlayerPosition)
    Projectile Test Length=3000
   ) {module=InitializeParticle_NDC_SW}
   &InitializeParticle(
    Lifetime=&Multiply_Float(A=&Divide_Float(A=$Particles.DistanceTraveled, B=$Emitter.Projectile Velocity), B=2)
    Ribbon Width=60
    Ribbon Width Mode=::NewEnumerator1
    UsePositionOffset=false
    Write Material Random=true
   )
   /SetVariables(Particles.LightColor=&SelectLinearColorFromArray(
     Array Sampling Mode=::NewEnumerator1
     Direct Array Index=$Particles.VariantIndex
     Color Selection Array=DataInterface() {propertyValues {ColorData [
        <rgba 1 0 0 1>
        <rgba 1 0.2 0 1>
        <rgba 1 0.5 0 1>
        <rgba 0.1269 1 0 1>
        <rgba 0 0.8333 1 1>
        <rgba 0 0.0833 1 1>
        <rgba 0.1623 0 1 1>
        <rgba 1 1 1 1>
       ]}}
     Random Seed=?
    )) {enabled=false, isSetParameters=true}
   &KillParticles(Kill Particles=$Particles.NearProjectile)
  ]
  particleUpdateScript [
   &ParticleState()
   /SetVariables(Particles.DirectedPosition:NiagaraPosition=&LerpPosition(
     A=&AddVectorToPosition(
      Position=$Particles.OriginPosition
      Vector=&Multiply_VectorByFloat(Vector=$Particles.VelocityDirection, Float=350)
     )
     B=&AddVectorToPosition(
      Position=$Particles.CurrentPosition
      Vector=&Multiply_VectorByFloat(Vector=$Particles.VelocityDirection, Float=1500)
     )
     Alpha=&FloatFromCurve(FloatCurve=DataInterface() {propertyValues {Curve {keys [
          {interpMode=RCIM_Cubic, leaveTangentWeight=0.3456, tangentMode=RCTM_User, time=0, value=0}
          {arriveTangent=2.0458, interpMode=RCIM_Cubic, leaveTangent=2.0458, tangentMode=RCTM_User, time=0.9, value=1}
         ]}}})
    )) {isSetParameters=true, module=SetVariables_#3}
   /SetVariables(Particles.TraveledAmount:NiagaraFloat=&VectorLength(Vector=&Subtract_Position(A=$Particles.DirectedPosition, B=$Particles.OriginPosition))) {isSetParameters=true, module=SetVariables_#4}
   /SetVariables(Particles.NormalizedTravel:NiagaraFloat=&Divide_Float(A=$Particles.TraveledAmount, B=$Particles.DistanceTraveled)) {isSetParameters=true, module=SetVariables_#5}
   /SetVariables(Particles.Position=&LerpPosition(
     A=$Particles.DirectedPosition
     B=&Subtract_Position(
      A=$Particles.DirectedPosition
      B=&Multiply_VectorByFloat(
       Vector=$Particles.VelocityDirection
       Float=&FloatFromCurve(
        FloatCurve=DataInterface() {propertyValues {Curve {keys [
            {arriveTangent=14.6174, interpMode=RCIM_Cubic, leaveTangent=14.6174, tangentMode=RCTM_User, time=0, value=0}
            {arriveTangentWeight=0.0833, interpMode=RCIM_Cubic, leaveTangentWeight=0.1833, tangentMode=RCTM_User, time=0.2, value=1}
            {arriveTangentWeight=0.3056, interpMode=RCIM_Cubic, leaveTangentWeight=0.1111, tangentMode=RCTM_User, time=0.8, value=1}
            {arriveTangent=-15.5626, tangentMode=RCTM_User, time=1, value=0}
           ]}}}
        CurveIndex=$Particles.NormalizedTravel
        Scale Curve=&ClampFloat(Float=$Particles.TraveledAmount, Min=$Particles.TraveledAmount, Max=$Emitter.BoltLength)
       )
      )
     )
     Alpha=$Particles.RibbonLinkOrder
    )) {isSetParameters=true, module=SetVariables_#6}
   &DynamicMaterialParameters(
    Index 0 Param 1=&RemapRange(
     Input Value To Remap=&MakeFloatFromInt(Int=$Particles.VariantIndex)
     Input Min=0
     Input Max=7
     Output Min=0
     Output Max=1
    )
    Index 0 Param 2=&SelectFloatFromArray(Direct Array Index=$Particles.VariantIndex, Float Selection Array=DataInterface() {propertyValues {FloatData [1,1,0.1,0.75,3,3,3,1]}})
    Module.Index0Float0Write=true
    Module.Index0Float1Write=true
    Module.Index0Float2Write=false
    Module.Index0Float3Write=false
   )
   /SetVariables(
    Particles.AttenuationRadius:NiagaraFloat=&MakeCustomFloatFromBool(
     Bool=&SetBoolByFloatComparison(A=$Particles.RibbonLinkOrder, B=0, Comparison Type=::NewEnumerator7)
     True Float=1000
     False Float=0
    )
    Particles.LightEnabled:NiagaraBool=&SetBoolByFloatComparison(A=$Particles.RibbonLinkOrder, B=0, Comparison Type=::NewEnumerator7)
   ) {enabled=false, isSetParameters=true, module=SetVariables_#7}
   &ScaleColor(
    Scale Alpha=&FloatFromCurve(
     FloatCurve=DataInterface() {propertyValues {Curve {keys [{arriveTangent=-2.2829, interpMode=RCIM_Cubic, leaveTangent=-2.2829, tangentMode=RCTM_User, time=0, value=1}, <at 1 0 RCIM_Cubic>]}}}
     CurveIndex=$Particles.NormalizedTravel
    )
    ScaleA=true
    ScaleRGB=true
   )
  ]
  rendererClasses [
   &NiagaraRibbonRendererProperties
   &NiagaraLightRendererProperties
   &NiagaraComponentRendererProperties
  ]
  renderers [
   {rendererClass=&NiagaraRibbonRendererProperties}
   {rendererClass=&NiagaraLightRendererProperties, rendererIndex=1}
   {rendererClass=&NiagaraComponentRendererProperties, rendererIndex=2}
  ]
 }
 Tracer_Ribbons_WeaponOwner {
  emitterSpawnScript [/SetVariables(Emitter.BoltLength=3000, Emitter.Projectile Velocity=37500) {isSetParameters=true, module=SetVariables_#2}]
  emitterUpdateScript [
   &EmitterState()
   &self:Spawn_NDC_SW(Instigator Filter=1, PositionType=2, Reader:NiagaraDataInterfaceDataChannelRead=$System.NDCReader, Spawn Count=2) {module=Spawn_NDC_SW}
  ]
  particleSpawnScript [
   &self:InitializeParticle_NDC_SW(
    Player Position:NiagaraPosition=&MakeVectorFromLinearColorRGB(Color=$NPC.FortniteNPC.PlayerPosition)
    Projectile Test Length=3000
   ) {module=InitializeParticle_NDC_SW}
   &InitializeParticle(
    Lifetime=&Multiply_Float(A=&Divide_Float(A=$Particles.DistanceTraveled, B=$Emitter.Projectile Velocity), B=2)
    Ribbon Width=60
    Ribbon Width Mode=::NewEnumerator1
    Sprite Size Mode=::NewEnumerator3
    Uniform Sprite Size=&MakeCustomFloatFromBool(Bool=&SetBoolByFloatComparison(A=$Particles.RibbonLinkOrder, B=0), True Float=0, False Float=10)
    UsePositionOffset=false
    Write Material Random=true
   )
   /SetVariables(Particles.LightColor=&SelectLinearColorFromArray(
     Array Sampling Mode=::NewEnumerator1
     Direct Array Index=$Particles.VariantIndex
     Color Selection Array=DataInterface() {propertyValues {ColorData [
        <rgba 1 0 0 1>
        <rgba 1 0.2 0 1>
        <rgba 1 0.5 0 1>
        <rgba 0.1269 1 0 1>
        <rgba 0 0.8333 1 1>
        <rgba 0 0.0833 1 1>
        <rgba 0.1623 0 1 1>
        <rgba 1 1 1 1>
       ]}}
     Random Seed=?
    )) {enabled=false, isSetParameters=true, module=SetVariables_#8}
   &KillParticles(Kill Particles=$Particles.NearProjectile)
  ]
  particleUpdateScript [
   &ParticleState()
   /SetVariables(Particles.DirectedPosition:NiagaraPosition=&LerpPosition(
     A=&AddVectorToPosition(
      Position=$Particles.OriginPosition
      Vector=&Multiply_VectorByFloat(Vector=$Particles.VelocityDirection, Float=350)
     )
     B=&AddVectorToPosition(
      Position=$Particles.CurrentPosition
      Vector=&Multiply_VectorByFloat(Vector=$Particles.VelocityDirection, Float=1500)
     )
     Alpha=&FloatFromCurve(FloatCurve=DataInterface() {propertyValues {Curve {keys [
          {interpMode=RCIM_Cubic, leaveTangentWeight=0.3456, tangentMode=RCTM_User, time=0, value=0}
          {arriveTangent=2.0458, interpMode=RCIM_Cubic, leaveTangent=2.0458, tangentMode=RCTM_User, time=0.9, value=1}
         ]}}})
    )) {isSetParameters=true, module=SetVariables_#3}
   /SetVariables(Particles.TraveledAmount:NiagaraFloat=&VectorLength(Vector=&Subtract_Position(A=$Particles.DirectedPosition, B=$Particles.OriginPosition))) {isSetParameters=true, module=SetVariables_#4}
   /SetVariables(Particles.NormalizedTravel:NiagaraFloat=&Divide_Float(A=$Particles.TraveledAmount, B=$Particles.DistanceTraveled)) {isSetParameters=true, module=SetVariables_#5}
   /SetVariables(Particles.Position=&LerpPosition(
     A=$Particles.DirectedPosition
     B=&Subtract_Position(
      A=$Particles.DirectedPosition
      B=&Multiply_VectorByFloat(
       Vector=$Particles.VelocityDirection
       Float=&FloatFromCurve(
        FloatCurve=DataInterface() {propertyValues {Curve {keys [
            {arriveTangentWeight=0.0833, interpMode=RCIM_Cubic, leaveTangentWeight=0.1833, tangentMode=RCTM_User, time=0.1, value=1}
            {arriveTangentWeight=0.3056, interpMode=RCIM_Cubic, leaveTangentWeight=0.1111, tangentMode=RCTM_User, time=0.8, value=1}
            {arriveTangent=-15.5626, tangentMode=RCTM_User, time=1, value=0}
           ]}}}
        CurveIndex=$Particles.NormalizedTravel
        Scale Curve=&ClampFloat(Float=$Particles.TraveledAmount, Min=$Particles.TraveledAmount, Max=$Emitter.BoltLength)
       )
      )
     )
     Alpha=$Particles.RibbonLinkOrder
    )) {isSetParameters=true, module=SetVariables_#6}
   &DynamicMaterialParameters(
    Index 0 Param 1=&RemapRange(
     Input Value To Remap=&MakeFloatFromInt(Int=$Particles.VariantIndex)
     Input Min=0
     Input Max=7
     Output Min=0
     Output Max=1
    )
    Index 0 Param 2=&SelectFloatFromArray(Direct Array Index=$Particles.VariantIndex, Float Selection Array=DataInterface() {propertyValues {FloatData [1,1,0.1,0.5,3,3,3,1]}})
    Module.Index0Float0Write=true
    Module.Index0Float1Write=true
    Module.Index0Float2Write=false
    Module.Index0Float3Write=false
   )
   /SetVariables(
    Particles.AttenuationRadius:NiagaraFloat=&MakeCustomFloatFromBool(
     Bool=&SetBoolByFloatComparison(A=$Particles.RibbonLinkOrder, B=0, Comparison Type=::NewEnumerator7)
     True Float=1000
     False Float=0
    )
    Particles.LightEnabled:NiagaraBool=&SetBoolByFloatComparison(A=$Particles.RibbonLinkOrder, B=0, Comparison Type=::NewEnumerator7)
   ) {enabled=false, isSetParameters=true, module=SetVariables_#7}
   /SetVariables(Particles.SpritePosition:NiagaraPosition=&AddVectorToPosition(Position=$Particles.Position, Vector=&Multiply_VectorByFloat(Vector=$Particles.VelocityDirection, Float=-45))) {isSetParameters=true, module=SetVariables_#9}
   &ScaleColor(
    Scale Alpha=&FloatFromCurve(
     FloatCurve=DataInterface() {propertyValues {Curve {keys [{arriveTangent=-2.2829, interpMode=RCIM_Cubic, leaveTangent=-2.2829, tangentMode=RCTM_User, time=0, value=1}, <at 1 0 RCIM_Cubic>]}}}
     CurveIndex=$Particles.NormalizedTravel
    )
    ScaleA=true
    ScaleRGB=true
   )
  ]
  rendererClasses [
   &NiagaraRibbonRendererProperties
   &NiagaraSpriteRendererProperties
   &NiagaraLightRendererProperties
   &NiagaraComponentRendererProperties
  ]
  renderers [
   {rendererClass=&NiagaraRibbonRendererProperties}
   {rendererClass=&NiagaraSpriteRendererProperties, rendererIndex=1}
   {rendererClass=&NiagaraLightRendererProperties, rendererIndex=2}
   {rendererClass=&NiagaraComponentRendererProperties, rendererIndex=3}
  ]
 }
 Tracer_Tracer {
  bEnabled=false
  emitterSpawnScript [/SetVariables(Emitter.BoltLength=500, Emitter.Projectile Velocity=75000) {isSetParameters=true, module=SetVariables_#2}]
  emitterUpdateScript [
   &EmitterState()
   &self:Spawn_NDC_SW(PositionType=2, Reader:NiagaraDataInterfaceDataChannelRead=$System.NDCReader, Spawn Count=2) {module=Spawn_NDC_SW}
  ]
  particleSpawnScript [
   &self:InitializeParticle_NDC_SW() {module=InitializeParticle_NDC_SW}
   &InitializeParticle(
    Color=&SelectLinearColorFromArray(
     Array Sampling Mode=::NewEnumerator1
     Direct Array Index=$Particles.VariantIndex
     Color Selection Array=DataInterface() {propertyValues {ColorData [
        <rgba 1 0 0 1>
        <rgba 1 0.2 0 1>
        <rgba 1 0.5 0 1>
        <rgba 0.1269 1 0 1>
        <rgba 0 0.8333 1 1>
        <rgba 0 0.0833 1 1>
        <rgba 0.1623 0 1 1>
        <rgba 1 1 1 1>
       ]}}
     Random Seed=?
    )
    Lifetime=&Multiply_Float(A=&Divide_Float(A=$Particles.DistanceTraveled, B=$Emitter.Projectile Velocity), B=5)
    Ribbon Width=40
    Ribbon Width Mode=::NewEnumerator1
    UsePositionOffset=false
    Write Material Random=true
   )
  ]
  particleUpdateScript [
   &ParticleState()
   /SetVariables(Particles.Position=&LerpPosition(
     A=&Lerp_Vector(
      A=Linked() {linkedVariable {name=Particles.OriginPosition, type=NiagaraPosition}}
      B=Linked() {linkedVariable {name=Particles.CurrentPosition, type=NiagaraPosition}}
      Alpha=&FloatFromCurve(FloatCurve=DataInterface() {propertyValues {Curve {keys [
           {arriveTangent=3.9168, interpMode=RCIM_Cubic, leaveTangent=3.9168, leaveTangentWeight=0.3456, tangentMode=RCTM_User, time=0.5, value=0}
           {interpMode=RCIM_Cubic, leaveTangent=1.6278, time=1, value=1}
          ]}}})
     )
     B=&Lerp_Vector(
      A=Linked() {linkedVariable {name=Particles.OriginPosition, type=NiagaraPosition}}
      B=&AddVectorToPosition(
       Position=$Particles.CurrentPosition
       Vector=&Multiply_VectorByFloat(Vector=$Particles.VelocityDirection, Float=1500)
      )
      Alpha=&FloatFromCurve(FloatCurve=DataInterface() {propertyValues {Curve {keys [{interpMode=RCIM_Cubic, leaveTangentWeight=0.3456, time=0, value=0}, {interpMode=RCIM_Cubic, leaveTangent=1.6278, time=1, value=1}]}}})
     )
     Alpha=$Particles.RibbonLinkOrder
    )) {isSetParameters=true, module=SetVariables_#6}
   &DynamicMaterialParameters(
    Index 0 Param 1=&RemapRange(
     Input Value To Remap=&MakeFloatFromInt(Int=$Particles.VariantIndex)
     Input Min=0
     Input Max=7
     Output Min=0
     Output Max=1
    )
    Index 0 Param 2=&SelectFloatFromArray(Direct Array Index=$Particles.VariantIndex, Float Selection Array=DataInterface() {propertyValues {FloatData [8,4,1,4,1,1,1,1]}})
    Module.Index0Float0Write=true
    Module.Index0Float1Write=true
    Module.Index0Float2Write=false
    Module.Index0Float3Write=false
   )
   &ScaleColor(
    Scale Alpha=&FloatFromCurve(FloatCurve=DataInterface() {propertyValues {Curve {keys [<at 0.75 1 RCIM_Cubic>,<at 1 0 RCIM_Cubic>]}}}, Scale Curve=0.0625)
    ScaleA=true
    ScaleRGB=true
   )
  ]
  rendererClasses [&NiagaraRibbonRendererProperties]
  renderers [{rendererClass=&NiagaraRibbonRendererProperties}]
 }
 Up_Close_and_Personal__Tracer {
  emitterSpawnScript [/SetVariables(Emitter.BoltLength=3500, Emitter.Projectile Velocity=37500) {isSetParameters=true, module=SetVariables_#2}]
  emitterUpdateScript [
   &EmitterState()
   &self:Spawn_NDC_SW(Instigator Filter=2, PositionType=2, Reader:NiagaraDataInterfaceDataChannelRead=$System.NDCReader, Spawn Count=2) {module=Spawn_NDC_SW}
  ]
  particleSpawnScript [
   &self:InitializeParticle_NDC_SW(
    Player Position:NiagaraPosition=&MakeVectorFromLinearColorRGB(Color=$NPC.FortniteNPC.PlayerPosition)
    Projectile Test Length=5000
   ) {module=InitializeParticle_NDC_SW}
   &InitializeParticle(Lifetime=0.1, Ribbon Width=60, Ribbon Width Mode=::NewEnumerator1, UsePositionOffset=false, Write Material Random=true)
   /SetVariables(Particles.LightColor=&SelectLinearColorFromArray(
     Array Sampling Mode=::NewEnumerator1
     Direct Array Index=$Particles.VariantIndex
     Color Selection Array=DataInterface() {propertyValues {ColorData [
        <rgba 1 0 0 1>
        <rgba 1 0.2 0 1>
        <rgba 1 0.5 0 1>
        <rgba 0.1269 1 0 1>
        <rgba 0 0.8333 1 1>
        <rgba 0 0.0833 1 1>
        <rgba 0.1623 0 1 1>
        <rgba 1 1 1 1>
       ]}}
     Random Seed=?
    )) {enabled=false, isSetParameters=true, module=SetVariables_#8}
   &KillParticles(Kill Particles=&Boolean_NotOperation(Bool=$Particles.NearProjectile))
  ]
  particleUpdateScript [
   &ParticleState()
   /SetVariables(Particles.DirectedPosition:NiagaraPosition=&LerpPosition(A=$Particles.OriginPosition, B=$Particles.CurrentPosition, Alpha=$Particles.NormalizedAge)) {isSetParameters=true, module=SetVariables_#3}
   /SetVariables(Particles.TraveledAmount:NiagaraFloat=&VectorLength(Vector=&Subtract_Position(A=$Particles.DirectedPosition, B=$Particles.OriginPosition))) {isSetParameters=true, module=SetVariables_#4}
   /SetVariables(Particles.NormalizedTravel:NiagaraFloat=&Divide_Float(A=$Particles.TraveledAmount, B=$Particles.DistanceTraveled)) {isSetParameters=true, module=SetVariables_#5}
   /SetVariables(Particles.Position=&LerpPosition(
     A=$Particles.DirectedPosition
     B=&Subtract_Position(
      A=$Particles.DirectedPosition
      B=&Multiply_VectorByFloat(
       Vector=$Particles.VelocityDirection
       Float=&FloatFromCurve(
        FloatCurve=DataInterface() {propertyValues {Curve {keys [
            {arriveTangentWeight=0.0833, interpMode=RCIM_Cubic, leaveTangentWeight=0.1833, tangentMode=RCTM_User, time=0.1, value=1}
            {arriveTangentWeight=0.3056, interpMode=RCIM_Cubic, leaveTangentWeight=0.1111, tangentMode=RCTM_User, time=0.8, value=1}
            {arriveTangent=-15.5626, tangentMode=RCTM_User, time=1, value=0}
           ]}}}
        CurveIndex=$Particles.NormalizedTravel
        Scale Curve=&ClampFloat(Float=$Particles.TraveledAmount, Min=$Particles.TraveledAmount, Max=$Emitter.BoltLength)
       )
      )
     )
     Alpha=$Particles.RibbonLinkOrder
    )) {isSetParameters=true, module=SetVariables_#6}
   &DynamicMaterialParameters(
    Index 0 Param 1=&RemapRange(
     Input Value To Remap=&MakeFloatFromInt(Int=$Particles.VariantIndex)
     Input Min=0
     Input Max=7
     Output Min=0
     Output Max=1
    )
    Index 0 Param 2=&SelectFloatFromArray(Direct Array Index=$Particles.VariantIndex, Float Selection Array=DataInterface() {propertyValues {FloatData [1,1,0.1,0.5,3,3,3,1]}})
    Module.Index0Float0Write=true
    Module.Index0Float1Write=true
    Module.Index0Float2Write=false
    Module.Index0Float3Write=false
   )
   /SetVariables(
    Particles.AttenuationRadius:NiagaraFloat=&MakeCustomFloatFromBool(
     Bool=&SetBoolByFloatComparison(A=$Particles.RibbonLinkOrder, B=0, Comparison Type=::NewEnumerator7)
     True Float=1000
     False Float=0
    )
    Particles.LightEnabled:NiagaraBool=&SetBoolByFloatComparison(A=$Particles.RibbonLinkOrder, B=0, Comparison Type=::NewEnumerator7)
   ) {enabled=false, isSetParameters=true, module=SetVariables_#7}
   &ScaleColor(
    Scale Alpha=&FloatFromCurve(
     FloatCurve=DataInterface() {propertyValues {Curve {keys [{arriveTangent=-2.2829, interpMode=RCIM_Cubic, time=0, value=1}, <at 1 0 RCIM_Cubic>]}}}
     CurveIndex=$Particles.NormalizedTravel
    )
    ScaleA=true
    ScaleRGB=true
   )
  ]
  rendererClasses [
   &NiagaraRibbonRendererProperties
   &NiagaraLightRendererProperties
   &NiagaraComponentRendererProperties
  ]
  renderers [
   {rendererClass=&NiagaraRibbonRendererProperties}
   {rendererClass=&NiagaraLightRendererProperties, rendererIndex=1}
   {rendererClass=&NiagaraComponentRendererProperties, rendererIndex=2}
  ]
 }
 Up_Close_and_Personal__Weapon {
  emitterSpawnScript [/SetVariables(Emitter.BoltLength=5000, Emitter.Projectile Velocity=37500) {isSetParameters=true, module=SetVariables_#2}]
  emitterUpdateScript [
   &EmitterState()
   &self:Spawn_NDC_SW(Instigator Filter=1, PositionType=2, Reader:NiagaraDataInterfaceDataChannelRead=$System.NDCReader, Spawn Count=2) {module=Spawn_NDC_SW}
  ]
  particleSpawnScript [
   &self:InitializeParticle_NDC_SW(
    Player Position:NiagaraPosition=&MakeVectorFromLinearColorRGB(Color=$NPC.FortniteNPC.PlayerPosition)
    Projectile Test Length=5000
   ) {module=InitializeParticle_NDC_SW}
   &InitializeParticle(
    Lifetime=0.1
    Ribbon Width=60
    Ribbon Width Mode=::NewEnumerator1
    Sprite Size Mode=::NewEnumerator3
    Uniform Sprite Size=&MakeCustomFloatFromBool(Bool=&SetBoolByFloatComparison(A=$Particles.RibbonLinkOrder, B=0), True Float=0, False Float=10)
    UsePositionOffset=false
    Write Material Random=true
   )
   /SetVariables(Particles.LightColor=&SelectLinearColorFromArray(
     Array Sampling Mode=::NewEnumerator1
     Direct Array Index=$Particles.VariantIndex
     Color Selection Array=DataInterface() {propertyValues {ColorData [
        <rgba 1 0 0 1>
        <rgba 1 0.2 0 1>
        <rgba 1 0.5 0 1>
        <rgba 0.1269 1 0 1>
        <rgba 0 0.8333 1 1>
        <rgba 0 0.0833 1 1>
        <rgba 0.1623 0 1 1>
        <rgba 1 1 1 1>
       ]}}
     Random Seed=?
    )) {isSetParameters=true, module=SetVariables_#8}
   &KillParticles(Kill Particles=&Boolean_NotOperation(Bool=$Particles.NearProjectile))
  ]
  particleUpdateScript [
   &ParticleState()
   /SetVariables(Particles.DirectedPosition:NiagaraPosition=&LerpPosition(A=$Particles.OriginPosition, B=$Particles.CurrentPosition, Alpha=$Particles.NormalizedAge)) {isSetParameters=true, module=SetVariables_#3}
   /SetVariables(Particles.TraveledAmount:NiagaraFloat=&VectorLength(Vector=&Subtract_Position(A=$Particles.DirectedPosition, B=$Particles.OriginPosition))) {isSetParameters=true, module=SetVariables_#4}
   /SetVariables(Particles.NormalizedTravel:NiagaraFloat=&Divide_Float(A=$Particles.TraveledAmount, B=$Particles.DistanceTraveled)) {isSetParameters=true, module=SetVariables_#5}
   /SetVariables(Particles.Position=&LerpPosition(
     A=$Particles.DirectedPosition
     B=&Subtract_Position(
      A=$Particles.DirectedPosition
      B=&Multiply_VectorByFloat(
       Vector=$Particles.VelocityDirection
       Float=&FloatFromCurve(
        FloatCurve=DataInterface() {propertyValues {Curve {keys [
            {arriveTangentWeight=0.0833, interpMode=RCIM_Cubic, leaveTangentWeight=0.1833, tangentMode=RCTM_User, time=0.1, value=1}
            {arriveTangentWeight=0.3056, interpMode=RCIM_Cubic, leaveTangentWeight=0.1111, tangentMode=RCTM_User, time=0.8, value=1}
            {arriveTangent=-15.5626, tangentMode=RCTM_User, time=1, value=0}
           ]}}}
        CurveIndex=$Particles.NormalizedTravel
        Scale Curve=&ClampFloat(Float=$Particles.TraveledAmount, Min=$Particles.TraveledAmount, Max=$Emitter.BoltLength)
       )
      )
     )
     Alpha=$Particles.RibbonLinkOrder
    )) {isSetParameters=true, module=SetVariables_#6}
   &DynamicMaterialParameters(
    Index 0 Param 1=&RemapRange(
     Input Value To Remap=&MakeFloatFromInt(Int=$Particles.VariantIndex)
     Input Min=0
     Input Max=7
     Output Min=0
     Output Max=1
    )
    Index 0 Param 2=&SelectFloatFromArray(Direct Array Index=$Particles.VariantIndex, Float Selection Array=DataInterface() {propertyValues {FloatData [1,1,0.1,0.5,3,3,3,1]}})
    Module.Index0Float0Write=true
    Module.Index0Float1Write=true
    Module.Index0Float2Write=false
    Module.Index0Float3Write=false
   )
   /SetVariables(
    Particles.AttenuationRadius:NiagaraFloat=&MakeCustomFloatFromBool(
     Bool=&SetBoolByFloatComparison(A=$Particles.RibbonLinkOrder, B=0, Comparison Type=::NewEnumerator7)
     True Float=1000
     False Float=0
    )
    Particles.LightEnabled:NiagaraBool=&SetBoolByFloatComparison(A=$Particles.RibbonLinkOrder, B=0, Comparison Type=::NewEnumerator7)
   ) {enabled=false, isSetParameters=true, module=SetVariables_#7}
   /SetVariables(Particles.SpritePosition:NiagaraPosition=&AddVectorToPosition(Position=$Particles.Position, Vector=&Multiply_VectorByFloat(Vector=$Particles.VelocityDirection, Float=-45))) {enabled=false, isSetParameters=true, module=SetVariables_#9}
   &ScaleColor(
    Scale Alpha=&FloatFromCurve(
     FloatCurve=DataInterface() {propertyValues {Curve {keys [{arriveTangent=-2.2829, interpMode=RCIM_Cubic, time=0, value=1}, <at 1 0 RCIM_Cubic>]}}}
     CurveIndex=$Particles.NormalizedTravel
    )
    ScaleA=true
    ScaleRGB=true
   )
  ]
  rendererClasses [
   &NiagaraRibbonRendererProperties
   &NiagaraSpriteRendererProperties
   &NiagaraLightRendererProperties
   &NiagaraComponentRendererProperties
  ]
  renderers [
   {rendererClass=&NiagaraRibbonRendererProperties}
   {rendererClass=&NiagaraSpriteRendererProperties, rendererIndex=1}
   {rendererClass=&NiagaraLightRendererProperties, rendererIndex=2}
   {rendererClass=&NiagaraComponentRendererProperties, rendererIndex=3}
  ]
 }
}
summary {
 systemName=NS_NDC_Jethro_ProjectilesHandler_HitScan_Shotguns
 emitters [
  {
   emitterName=Tracer_Ribbons
   rendererClasses [
    &NiagaraRibbonRendererProperties
    &NiagaraLightRendererProperties
    &NiagaraComponentRendererProperties
   ]
  }
  {
   emitterName=Tracer_Ribbons_WeaponOwner
   rendererClasses [
    &NiagaraRibbonRendererProperties
    &NiagaraSpriteRendererProperties
    &NiagaraLightRendererProperties
    &NiagaraComponentRendererProperties
   ]
  }
  {bEnabled=false, emitterName=Tracer_Tracer, rendererClasses [&NiagaraRibbonRendererProperties]}
  {
   emitterName=Up_Close_and_Personal__Weapon
   rendererClasses [
    &NiagaraRibbonRendererProperties
    &NiagaraSpriteRendererProperties
    &NiagaraLightRendererProperties
    &NiagaraComponentRendererProperties
   ]
  }
  {
   emitterName=Up_Close_and_Personal__Tracer
   rendererClasses [
    &NiagaraRibbonRendererProperties
    &NiagaraLightRendererProperties
    &NiagaraComponentRendererProperties
   ]
  }
 ]
 userVariables []
}
