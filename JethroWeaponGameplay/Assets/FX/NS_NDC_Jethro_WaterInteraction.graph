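# Niagara system that plays water impact effects for the Jethro weapon, spawned from Data Channel reads.
# A shared initializer supplies impact point, normal, water data and ammo type; each emitter derives a scale from ammo type and impact multiplier.
# A variant-colored flash pairs with gravity-driven drops and SubUV splashes launched along the impact normal.
# A foam mesh stays constrained to the water plane and passes water body, depth and wave attenuation to its material.
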NiagaraSystem /JethroWeaponGameplay/Assets/FX/NS_NDC_Jethro_WaterInteraction
dependencies {
 usedDataInterfaces [
  &NiagaraDataInterfaceArrayFloat
  &NiagaraDataInterfaceCurve
  &NiagaraDataInterfaceSpriteRendererInfo
  &NiagaraDataInterfaceArrayColor
  /Script/Niagara.NiagaraDataInterfaceDataChannelRead
 ]
 usedDynamicInputs [&UniformRangedVector2D]
 usedModules [
  &SystemState
  &self:SystemSpawnScript.SystemScriptSource.SystemScriptGraph.NiagaraNodeAssignment_0.SetVariables_#1
  /Niagara/Modules/System/CompleteIfUnused
 ]
}
emitters {
 Base_Flash {
  emitterUpdateScript [
   &EmitterState()
   /Game/Effects/Niagara/DataChannels/Impacts/Spawn_NDCFortImpacts(Max Spawn Count=1, Min Spawn Count=1, Reader:NiagaraDataInterfaceDataChannelRead=$System.NDCReader)
  ]
  particleSpawnScript [
   &self:Initialize_Particles_from_NDC_Water_Interaction_Jethro(FallbackScaleMultiplier=50) {module=Initialize_Particles_from_NDC_Water_Interaction_Jethro}
   /SetVariables(Particles.AmmoTypeScale:NiagaraFloat=&Multiply_Float(
     A=&SelectFloatFromArray(
      Direct Array Index=$Output.Initialize_Particles_from_NDC_Water_Interaction_Jethro.AmmoType
      Float Selection Array=DataInterface() {propertyValues {FloatData [1,1.4,1.7,0.95,1]}}
     )
     B=$Output.Initialize_Particles_from_NDC_Water_Interaction_Jethro.ImpactScaleMultiplier
    )) {isSetParameters=true, module=SetVariables_#2}
   &InitializeParticle(
    Color=&SelectLinearColorFromArray(
     Array Sampling Mode=::NewEnumerator1
     Direct Array Index=$Particles.VariantIndex
     Color Selection Array=DataInterface() {propertyValues {ColorData [
        <rgba 1 0 0 1>
        <rgba 1 0.2 0 1>
        <rgba 1 0.5 0 1>
        <rgba 0.1269 1 0 1>
        <rgba 0 0.8333 1 1>
        <rgba 0 0.36 3 1>
        <rgba 0.1623 0 1 1>
        <rgba 1 1 1 1>
       ]}}
     Random Seed=?
    )
    Lifetime Max=0.18
    Lifetime Min=0.13
    Lifetime Mode=::NewEnumerator1
    Mesh Scale=&Multiply_VectorByFloat(
     Vector=&RandomRangeVector(Minimum=<0.1 0.2 0.15>, Maximum=<0.1 0.45 0.2>)
     Float=$Particles.AmmoTypeScale
    )
    Mesh Scale Mode=::NewEnumerator4
    Position=$Output.Initialize_Particles_from_NDC_Water_Interaction_Jethro.ImpactPoint
    Position Mode=::NewEnumerator1
    Sprite Size Mode=::NewEnumerator3
    Uniform Sprite Size=&Multiply_Float(A=&RandomRangeFloat(Minimum=112.5, Maximum=187.5), B=$Particles.AmmoTypeScale)
    UsePositionOffset=false
    Write Material Random=true
   )
   &CameraOffset(Camera Offset Amount=10)
   /SetVariables() {isSetParameters=true}
  ]
  particleUpdateScript [
   &ParticleState()
   &ScaleMeshSize(Scale Factor=&VectorFromFloat(Value=&FloatFromCurve(FloatCurve=DataInterface() {propertyValues {Curve {keys [
          {arriveTangent=9.901, interpMode=RCIM_Cubic, leaveTangent=9.901, tangentMode=RCTM_User, time=0, value=0.4}
          {arriveTangent=12.2238, interpMode=RCIM_Cubic, leaveTangent=12.2238, tangentMode=RCTM_User, time=1, value=6}
         ]}}})))
   &ScaleColor(
    Scale Alpha=&FloatFromCurve(FloatCurve=DataInterface() {propertyValues {Curve {keys [<at 0.7 1 RCIM_Cubic>,<at 1 0>]}}}, Scale Curve=0.25)
    Scale RGB=<2 2 2>
    ScaleA=true
    ScaleRGB=true
   )
   /SetVariables(
    Particles.Glow Sprite Camera Offset=15
    Particles.Glow Sprite Color:LinearColor=&MultiplyLinearColor(
     Linear Color=&SelectLinearColorFromArray(
      Array Sampling Mode=::NewEnumerator1
      Direct Array Index=$Particles.VariantIndex
      Color Selection Array=DataInterface() {propertyValues {ColorData [
         <rgba 1 0 0 1>
         <rgba 1 0.2 0 1>
         <rgba 1 0.5 0 1>
         <rgba 0.1269 1 0 1>
         <rgba 0 0.8333 1 1>
         <rgba 0 0.0833 1 1>
         <rgba 0.1623 0 1 1>
         <rgba 1 1 1 1>
        ]}}
      Random Seed=?
     )
     Scale Factor=&MakeFloat4FromVectorAndFloat(
      "Vector (RGB)"=<1.5 1.5 1.5>
      "Float (Alpha)"=&FloatFromCurve(FloatCurve=DataInterface() {propertyValues {Curve {keys [
           {arriveTangent=8.356, interpMode=RCIM_Cubic, leaveTangent=8.356, tangentMode=RCTM_User, time=0, value=0}
           {arriveTangentWeight=0.2855, interpMode=RCIM_Cubic, leaveTangentWeight=0.1894, tangentMode=RCTM_User, time=0.2, value=1}
           {arriveTangentWeight=0.6992, interpMode=RCIM_Cubic, tangentMode=RCTM_User, time=1, value=0}
          ]}}})
     )
    )
    Particles.Glow Sprite Rotation:NiagaraFloat=&Add_Float(
     A=$Particles.Glow Sprite Rotation
     B=&Multiply_Float(
      A=&FloatFromCurve(FloatCurve=DataInterface() {propertyValues {Curve {keys [
           {tangentMode=RCTM_User, time=0, value=1}
           {arriveTangent=-0.3174, interpMode=RCIM_Cubic, leaveTangent=-0.3174, tangentMode=RCTM_User, time=0.3, value=1}
           {arriveTangent=0.0123, interpMode=RCIM_Cubic, leaveTangent=0.0123, tangentMode=RCTM_User, time=1, value=0}
          ]}}})
      B=-5
     )
    )
    Particles.Glow Sprite Size:Vector2f=&Vector2DFromFloat(Value=&Multiply_Float(
      A=&Multiply_Float(A=145, B=$Particles.AmmoTypeScale)
      B=&FloatFromCurve(FloatCurve=DataInterface() {propertyValues {Curve {
          defaultValue=0
          keys [
           {arriveTangent=0.6169, interpMode=RCIM_Cubic, time=0, value=0}
           {arriveTangent=0.0666, interpMode=RCIM_Cubic, leaveTangent=0.0666, tangentMode=RCTM_User, time=0.5, value=0.9847}
           {arriveTangent=-4.3124, interpMode=RCIM_Cubic, leaveTangent=-4.3124, tangentMode=RCTM_User, time=1, value=0}
          ]
         }}})
     ))
   ) {isSetParameters=true}
  ]
  rendererClasses [&NiagaraSpriteRendererProperties, &NiagaraMeshRendererProperties]
  renderers [
   {rendererClass=&NiagaraSpriteRendererProperties}
   {rendererClass=&NiagaraMeshRendererProperties, rendererIndex=1}
  ]
 }
 Drops_Tall {
  emitterUpdateScript [
   &EmitterState()
   &Spawn_NDCFortImpacts(Max Spawn Count=5, Min Spawn Count=3, Reader:NiagaraDataInterfaceDataChannelRead=$System.NDCReader)
  ]
  particleSpawnScript [
   &self:Initialize_Particles_from_NDC_Water_Interaction_Jethro(FallbackScaleMultiplier=50) {module=Initialize_Particles_from_NDC_Water_Interaction_Jethro}
   /SetVariables(Particles.AmmoTypeScale:NiagaraFloat=&Multiply_Float(
     A=&SelectFloatFromArray(
      Direct Array Index=$Output.Initialize_Particles_from_NDC_Water_Interaction_Jethro.AmmoType
      Float Selection Array=DataInterface() {propertyValues {FloatData [1,1.4,1.7,0.95,1]}}
     )
     B=$Output.Initialize_Particles_from_NDC_Water_Interaction_Jethro.ImpactScaleMultiplier
    )) {isSetParameters=true, module=SetVariables_#3}
   &InitializeParticle(
    Lifetime=&UniformRangedFloat(Minimum=0.2, Maximum=1)
    Mass Mode=::NewEnumerator1
    Position=$Output.Initialize_Particles_from_NDC_Water_Interaction_Jethro.ImpactPoint
    Position Mode=::NewEnumerator1
    Sprite Size Mode=::NewEnumerator1
    Uniform Sprite Size Max=6
    UsePositionOffset=false
    Write Material Random=true
   )
   &AddVelocityInCone(
    Cone Angle=20
    Cone Axis=$Output.Initialize_Particles_from_NDC_Water_Interaction_Jethro.ImpactNormal
    Module.Use Velocity Falloff On Cone Axis=false
    Velocity Strength=&Multiply_Float(
     A=&Multiply_Float(A=&UniformRangedFloat(Minimum=375, Maximum=475), B=$Particles.AmmoTypeScale)
     B=&FloatFromCurve(
      FloatCurve=DataInterface() {propertyValues {Curve {keys [{arriveTangent=0.6429, interpMode=RCIM_Cubic, leaveTangent=0.6429, tangentMode=RCTM_User, time=0, value=0.5}, <at 1 1 RCIM_Cubic>]}}}
      CurveIndex=&ReturnNormalizedExecIndex()
     )
    )
   )
  ]
  particleUpdateScript [
   &ParticleState()
   &GravityForce(Gravity=&Multiply_VectorByFloat(Vector=<0 0 -980>, Float=1))
   &SolveForcesAndVelocity(Clamp Velocity=false, Limit Acceleration=false)
   &ScaleSpriteSize(
    Uniform Curve Scale=&FloatFromCurve(
     FloatCurve=DataInterface() {propertyValues {Curve {keys [{arriveTangent=1.1621, interpMode=RCIM_Cubic, time=0, value=0.66}, <at 1 1 RCIM_Cubic>]}}}
     CurveIndex=&ReturnNormalizedExecIndex()
    )
    Uniform Curve Sprite Scale=DataInterface() {propertyValues {Curve {keys [
        {tangentMode=RCTM_User, time=0, value=1}
        {arriveTangent=-0.0405, interpMode=RCIM_Cubic, leaveTangent=-0.0405, tangentMode=RCTM_User, time=0.7009, value=1}
        {arriveTangent=-8.0089, interpMode=RCIM_Cubic, leaveTangent=-8.0089, tangentMode=RCTM_User, time=1, value=0}
       ]}}}
   )
   &DynamicMaterialParameters(
    Index 0 Param 2=&FloatFromCurve(FloatCurve=DataInterface() {propertyValues {Curve {keys [{interpMode=RCIM_Cubic, tangentMode=RCTM_User, time=0, value=1}, {interpMode=RCIM_Cubic, tangentMode=RCTM_User, time=1, value=0}]}}})
    Module.Index0Param0Write=true
    Module.Index0Param1Write=true
    Module.Index0Param2Write=true
    Module.Index0Param3Write=true
   )
   &Color(Color=&MakeLinearColorFromVectorAndFloat(
     "Vector (RGB)"=<1 1 1>
     "Float (Alpha)"=&FloatFromCurve(FloatCurve=DataInterface() {propertyValues {Curve {keys [<at 0.4 1.0113 RCIM_Cubic>,<at 1.0011 0.002>]}}})
    ))
  ]
  rendererClasses [&NiagaraSpriteRendererProperties]
  renderers [{rendererClass=&NiagaraSpriteRendererProperties}]
 }
 Drops_Wide001 {
  emitterUpdateScript [
   &EmitterState()
   &Spawn_NDCFortImpacts(Max Spawn Count=13, Min Spawn Count=9, Reader:NiagaraDataInterfaceDataChannelRead=$System.NDCReader)
  ]
  particleSpawnScript [
   &self:Initialize_Particles_from_NDC_Water_Interaction_Jethro(FallbackScaleMultiplier=50) {module=Initialize_Particles_from_NDC_Water_Interaction_Jethro}
   /SetVariables(Particles.AmmoTypeScale:NiagaraFloat=&Multiply_Float(
     A=&SelectFloatFromArray(
      Direct Array Index=$Output.Initialize_Particles_from_NDC_Water_Interaction_Jethro.AmmoType
      Float Selection Array=DataInterface() {propertyValues {FloatData [1,1.2,1.7,0.95,1]}}
     )
     B=$Output.Initialize_Particles_from_NDC_Water_Interaction_Jethro.ImpactScaleMultiplier
    )) {isSetParameters=true, module=SetVariables_#3}
   &InitializeParticle(
    Lifetime=&UniformRangedFloat(Minimum=0.2, Maximum=0.6)
    Mesh Scale Mode=::NewEnumerator3
    Mesh Uniform Scale=4
    Position=$Output.Initialize_Particles_from_NDC_Water_Interaction_Jethro.ImpactPoint
    Position Mode=::NewEnumerator1
    Sprite Size Mode=::NewEnumerator1
    Uniform Sprite Size Max=5
    Uniform Sprite Size Min=2
    UsePositionOffset=false
    Write Material Random=true
   )
   &AddVelocityInCone(
    Cone Angle=&RandomRangeFloat(Minimum=45, Maximum=70)
    Cone Axis=$Output.Initialize_Particles_from_NDC_Water_Interaction_Jethro.ImpactNormal
    Module.Use Velocity Falloff On Cone Axis=false
    Velocity Strength=&Multiply_Float(A=&UniformRangedFloat(Minimum=275, Maximum=425), B=$Particles.AmmoTypeScale)
   )
  ]
  particleUpdateScript [
   &ParticleState()
   &GravityForce(Gravity=&Multiply_VectorByFloat(Vector=<0 0 -980>, Float=1))
   &SolveForcesAndVelocity(Clamp Velocity=false, Limit Acceleration=false)
   &ScaleSpriteSize(Uniform Curve Sprite Scale=DataInterface() {propertyValues {Curve {keys [
        {tangentMode=RCTM_User, time=0, value=1}
        {arriveTangent=-0.0405, interpMode=RCIM_Cubic, leaveTangent=-0.0405, tangentMode=RCTM_User, time=0.7009, value=1}
        {arriveTangent=-8.0089, interpMode=RCIM_Cubic, leaveTangent=-8.0089, tangentMode=RCTM_User, time=1, value=0}
       ]}}})
   &ScaleSpriteSizeBySpeed(Max Scale Factor=<0.5 4>, Velocity Threshold=500)
   &DynamicMaterialParameters(
    Index 0 Param 2=&FloatFromCurve(FloatCurve=DataInterface() {propertyValues {Curve {keys [{interpMode=RCIM_Cubic, tangentMode=RCTM_User, time=0, value=1}, {interpMode=RCIM_Cubic, tangentMode=RCTM_User, time=1, value=0}]}}})
    Module.Index0Float0Write=true
    Module.Index0Float1Write=true
    Module.Index0Float2Write=true
    Module.Index0Float3Write=true
   )
   &Color(Color=&MakeLinearColorFromVectorAndFloat(
     "Vector (RGB)"=<1 1 1>
     "Float (Alpha)"=&FloatFromCurve(FloatCurve=DataInterface() {propertyValues {Curve {keys [<at 0.4 1.0113 RCIM_Cubic>,<at 1.0011 0.002>]}}})
    ))
  ]
  rendererClasses [&NiagaraSpriteRendererProperties]
  renderers [{rendererClass=&NiagaraSpriteRendererProperties}]
 }
 Impact_Surface_Foam {
  emitterUpdateScript [
   &EmitterState()
   &Spawn_NDCFortImpacts(Max Spawn Count=1, Min Spawn Count=1, Reader:NiagaraDataInterfaceDataChannelRead=$System.NDCReader)
  ]
  particleSpawnScript [
   &self:Initialize_Particles_from_NDC_Water_Interaction_Jethro(FallbackScaleMultiplier=50) {module=Initialize_Particles_from_NDC_Water_Interaction_Jethro}
   /SetVariables(Particles.AmmoTypeScale:NiagaraFloat=&Multiply_Float(
     A=&SelectFloatFromArray(
      Direct Array Index=$Output.Initialize_Particles_from_NDC_Water_Interaction_Jethro.AmmoType
      Float Selection Array=DataInterface() {propertyValues {FloatData [0.8,1.4,1.4,0.6,1]}}
     )
     B=$Output.Initialize_Particles_from_NDC_Water_Interaction_Jethro.ImpactScaleMultiplier
    )) {isSetParameters=true}
   &InitializeParticle(
    Lifetime Max=1.6
    Lifetime Min=1.5
    Lifetime Mode=::NewEnumerator1
    Mass=&UniformRangedFloat(Minimum=0.6, Maximum=1.2)
    Mass Mode=::NewEnumerator1
    Mesh Scale Mode=::NewEnumerator3
    Mesh Uniform Scale=&Multiply_Float(A=200, B=$Particles.AmmoTypeScale)
    Position=$Output.Initialize_Particles_from_NDC_Water_Interaction_Jethro.ImpactPoint
    Position Mode=::NewEnumerator1
    Position Offset=&Add_Vector(A=$Output.Initialize_Particles_from_NDC_Water_Interaction_Jethro.WaveOffset, B=<0 0 2>)
    Position Offset Coordinate Space=::World
    UsePositionOffset=true
    Write Material Random=true
   )
   &InitialMeshOrientation(
    Mesh Orientation Mode=::NewEnumerator2
    Orientation Vector=<0 0 0>
    Rotation=&UniformRangedVector(Minimum=<0 0 0>, Maximum=<0 0 1>)
    Use Rotation Vector=false
   ) {module=InitialMeshOrientation001}
   &DynamicMaterialParameters(
    Index 3 Param 1=&MakeFloatFromInt(Int=$Output.Initialize_Particles_from_NDC_Water_Interaction_Jethro.WaterBodyIndex)
    Index 3 Param 2=$Output.Initialize_Particles_from_NDC_Water_Interaction_Jethro.WaterDepth
    Index 3 Param 3=$Output.Initialize_Particles_from_NDC_Water_Interaction_Jethro.WaveAttenuationFactor
    Module.Index3Param0Write=true
    Module.Index3Param1Write=true
    Module.Index3Param2Write=true
    Module.Index3Param3Write=true
    Write Parameter Index 0=false
    Write Parameter Index 3=true
   )
  ]
  particleUpdateScript [
   &ParticleState()
   &DynamicMaterialParameters(
    Index 0 Param 1=&RandomRangeFloat(Minimum=-1, Maximum=1)
    Index 0 Param 2=&FloatFromCurve(FloatCurve=DataInterface() {propertyValues {Curve {keys [<at 0 0.33>,<at 1 1.25>]}}})
    Index 0 Param 3=&FloatFromCurve(
     FloatCurve=DataInterface() {propertyValues {Curve {keys [{tangentMode=RCTM_User, time=0, value=1}, {tangentMode=RCTM_User, time=1, value=0}]}}}
     Scale Curve=1.2
    )
    Index 1 Param 1=&MakeFloatFromVector(Vector=Linked() {linkedVariable {name=Particles.Initial.Position, type=NiagaraPosition}}, Channel=::NewEnumerator0)
    Index 1 Param 2=&MakeFloatFromVector(Vector=Linked() {linkedVariable {name=Particles.Initial.Position, type=NiagaraPosition}}, Channel=::NewEnumerator1)
    Index 1 Param 3=&FloatFromCurve(
     FloatCurve=DataInterface() {propertyValues {Curve {keys [<at 0 0.8>, {arriveTangent=0.579, leaveTangent=0.579, time=1, value=1}]}}}
     Scale Curve=1600
    )
    Index 2 Param 1=&FloatFromCurve(FloatCurve=DataInterface() {propertyValues {Curve {keys [{interpMode=RCIM_Cubic, tangentMode=RCTM_User, time=0, value=0.4}, {interpMode=RCIM_Cubic, tangentMode=RCTM_User, time=1, value=1}]}}})
    Index 2 Param 2=&FloatFromCurve(FloatCurve=DataInterface() {propertyValues {Curve {keys [<at 0 2>,<at 1 0>]}}})
    Index 2 Param 3=1
    Index 2 Param 4=&FloatFromCurve(FloatCurve=DataInterface() {propertyValues {Curve {keys [
         {arriveTangent=-1.9991, interpMode=RCIM_Cubic, time=0, value=1}
         {arriveTangent=-2.7143, interpMode=RCIM_Cubic, leaveTangent=-2.7143, tangentMode=RCTM_User, time=0.2095, value=0.6901}
         <at 1 0 RCIM_Cubic>
        ]}}})
    Module.Index0Param0Write=true
    Module.Index0Param1Write=true
    Module.Index0Param2Write=true
    Module.Index0Param3Write=true
    Module.Index1Param0Write=true
    Module.Index1Param1Write=true
    Module.Index1Param2Write=true
    Module.Index1Param3Write=true
    Module.Index2Param0Write=true
    Module.Index2Param1Write=true
    Module.Index2Param2Write=true
    Module.Index2Param3Write=true
    Write Parameter Index 1=true
    Write Parameter Index 2=true
   ) {module=DynamicMaterialParameters001}
   &ScaleMeshSize(Scale Factor=&VectorFromFloat(Value=&FloatFromCurve(FloatCurve=DataInterface() {propertyValues {Curve {keys [{arriveTangent=0.9849, interpMode=RCIM_Cubic, leaveTangent=0.9849, tangentMode=RCTM_User, time=0, value=0.33}, <at 1 1 RCIM_Cubic>]}}}))) {module=MeshSizeScale}
   &Color(Color=&MakeLinearColorFromVectorAndFloat(
     "Vector (RGB)"=<1 1 1>
     "Float (Alpha)"=&FloatFromCurve(FloatCurve=DataInterface() {propertyValues {Curve {keys [<at 0.4 1.0113 RCIM_Cubic>,<at 1.0011 0.002>]}}})
    ))
   &ConstrainPositionToPlane(
    Plane Normal:Vector3f=$Output.Initialize_Particles_from_NDC_Water_Interaction_Jethro.WaterPlaneNormal
    Plane Position:NiagaraPosition=$Output.Initialize_Particles_from_NDC_Water_Interaction_Jethro.WaterPlaneLocation
   )
  ]
  rendererClasses [&NiagaraMeshRendererProperties]
  renderers [{rendererClass=&NiagaraMeshRendererProperties, rendererIndex=0}]
 }
 Splash {
  emitterUpdateScript [
   &EmitterState()
   &Spawn_NDCFortImpacts(Max Spawn Count=2, Min Spawn Count=1, Reader:NiagaraDataInterfaceDataChannelRead=$System.NDCReader)
  ]
  particleSpawnScript [
   &self:Initialize_Particles_from_NDC_Water_Interaction_Jethro(FallbackScaleMultiplier=50) {module=Initialize_Particles_from_NDC_Water_Interaction_Jethro}
   /SetVariables(Particles.AmmoTypeScale:NiagaraFloat=&Multiply_Float(
     A=&SelectFloatFromArray(
      Direct Array Index=$Output.Initialize_Particles_from_NDC_Water_Interaction_Jethro.AmmoType
      Float Selection Array=DataInterface() {propertyValues {FloatData [1,1.4,1.7,0.95,1]}}
     )
     B=$Output.Initialize_Particles_from_NDC_Water_Interaction_Jethro.ImpactScaleMultiplier
    )) {isSetParameters=true, module=SetVariables_#2}
   &InitializeParticle(
    Lifetime Max=0.5
    Lifetime Min=0.4
    Lifetime Mode=::NewEnumerator1
    Position=$Output.Initialize_Particles_from_NDC_Water_Interaction_Jethro.ImpactPoint
    Position Mode=::NewEnumerator1
    Sprite Rotation Mode=::NewEnumerator2
    Sprite Size=&Multiply_Vector2D_ByFloat(
     Float=&Multiply_Float(A=&RandomRangeFloat(Minimum=0.8, Maximum=1), B=$Particles.AmmoTypeScale)
     Vector2D=<155 155>
    )
    Sprite Size Mode=::NewEnumerator4
    UsePositionOffset=true
    Write Material Random=true
   )
   &AddVelocityInCone(
    Cone Angle=3
    Cone Axis=$Output.Initialize_Particles_from_NDC_Water_Interaction_Jethro.ImpactNormal
    Module.Use Velocity Falloff On Cone Axis=false
    Velocity Strength=&Multiply_Float(A=&UniformRangedFloat(Minimum=185, Maximum=240), B=$Particles.AmmoTypeScale)
   )
   &SubUVAnimation(
    Evaluation Type=::NewEnumerator0
    "Sprite Renderer "=DataInterface() {propertyValues {SpriteRenderer=&self:Splash_4.NiagaraSpriteRendererProperties_0}}
    SubUV Animation Mode:ENiagara_SubUVLookupModeV2=::NewEnumerator2
    UseEndFrame=false
    UseStartFrame=false
   ) {module=SubUVAnimation001}
  ]
  particleUpdateScript [
   &ParticleState()
   &GravityForce(Gravity=&Multiply_VectorByFloat(Vector=<0 0 -980>, Float=1))
   &SolveForcesAndVelocity(Clamp Velocity=false, Limit Acceleration=false)
   &ScaleSpriteSize(Uniform Curve Sprite Scale=DataInterface() {propertyValues {Curve {keys [{tangentMode=RCTM_User, time=0, value=0.2}, {tangentMode=RCTM_User, time=1, value=1}]}}})
   &DynamicMaterialParameters(
    Index 0 Param 2=&FloatFromCurve(FloatCurve=DataInterface() {propertyValues {Curve {keys [{interpMode=RCIM_Cubic, tangentMode=RCTM_User, time=0, value=1}, {interpMode=RCIM_Cubic, tangentMode=RCTM_User, time=1, value=0}]}}})
    Index 0 Param 4=&FloatFromCurve(FloatCurve=DataInterface() {propertyValues {Curve {keys [{tangentMode=RCTM_User, time=0, value=0}, {tangentMode=RCTM_User, time=1, value=1}]}}})
    Module.Index0Float0Write=true
    Module.Index0Float1Write=true
    Module.Index0Float2Write=true
    Module.Index0Float3Write=true
   )
   &Color(Color=&MakeLinearColorFromVectorAndFloat(
     "Vector (RGB)"=<1 1 1>
     "Float (Alpha)"=&FloatFromCurve(FloatCurve=DataInterface() {propertyValues {Curve {keys [<at 0.4 1 RCIM_Cubic>,<at 1 0>]}}})
    ))
  ]
  rendererClasses [&NiagaraSpriteRendererProperties]
  renderers [{rendererClass=&NiagaraSpriteRendererProperties}]
 }
 Splash001 {
  emitterUpdateScript [
   &EmitterState()
   &Spawn_NDCFortImpacts(Max Spawn Count=1, Min Spawn Count=1, Reader:NiagaraDataInterfaceDataChannelRead=$System.NDCReader)
  ]
  particleSpawnScript [
   &self:Initialize_Particles_from_NDC_Water_Interaction_Jethro(FallbackScaleMultiplier=50) {module=Initialize_Particles_from_NDC_Water_Interaction_Jethro}
   /SetVariables(Particles.AmmoTypeScale:NiagaraFloat=&Multiply_Float(
     A=&SelectFloatFromArray(
      Direct Array Index=$Output.Initialize_Particles_from_NDC_Water_Interaction_Jethro.AmmoType
      Float Selection Array=DataInterface() {propertyValues {FloatData [1,1.4,1.7,0.95,1]}}
     )
     B=$Output.Initialize_Particles_from_NDC_Water_Interaction_Jethro.ImpactScaleMultiplier
    )) {isSetParameters=true, module=SetVariables_#2}
   &InitializeParticle(
    Lifetime Max=0.4
    Lifetime Min=0.3
    Lifetime Mode=::NewEnumerator1
    Position=$Output.Initialize_Particles_from_NDC_Water_Interaction_Jethro.ImpactPoint
    Position Mode=::NewEnumerator1
    Position Offset=<0 0 2>
    Sprite Size=&Multiply_Vector2D_ByFloat(
     Float=&Multiply_Float(A=&RandomRangeFloat(Minimum=0.8, Maximum=1), B=$Particles.AmmoTypeScale)
     Vector2D=<30 95>
    )
    Sprite Size Mode=::NewEnumerator4
    UsePositionOffset=true
    Write Material Random=true
   )
   &AddVelocityInCone(
    Cone Angle=5
    Cone Axis=$Output.Initialize_Particles_from_NDC_Water_Interaction_Jethro.ImpactNormal
    Module.Use Velocity Falloff On Cone Axis=false
    Velocity Strength=&Multiply_Float(A=&UniformRangedFloat(Minimum=50, Maximum=100), B=$Particles.AmmoTypeScale)
   )
   &SubUVAnimation(
    Evaluation Type=::NewEnumerator0
    "Sprite Renderer "=DataInterface() {propertyValues {SpriteRenderer=&self:Splash001_6.NiagaraSpriteRendererProperties_0}}
    SubUV Animation Mode:ENiagara_SubUVLookupModeV2=::NewEnumerator2
    UseEndFrame=false
    UseStartFrame=false
   ) {module=SubUVAnimation001}
   &SpriteFacingAndAlignment(Sprite Alignment=$Particles.Velocity, WriteAlignment=true, WriteFacing=false)
  ]
  particleUpdateScript [
   &ParticleState()
   &GravityForce(Gravity=&Multiply_VectorByFloat(Vector=<0 0 -980>, Float=1))
   &SolveForcesAndVelocity(Clamp Velocity=false, Limit Acceleration=false)
   &ScaleSpriteSize(Uniform Curve Sprite Scale=DataInterface() {propertyValues {Curve {keys [{arriveTangent=2.2272, interpMode=RCIM_Cubic, leaveTangent=2.2272, tangentMode=RCTM_User, time=0, value=0.33}, <at 0.5 1 RCIM_Cubic>]}}})
   &DynamicMaterialParameters(
    Index 0 Param 2=&FloatFromCurve(FloatCurve=DataInterface() {propertyValues {Curve {keys [{interpMode=RCIM_Cubic, tangentMode=RCTM_User, time=0, value=1}, {interpMode=RCIM_Cubic, tangentMode=RCTM_User, time=1, value=0}]}}})
    Index 0 Param 4=&FloatFromCurve(FloatCurve=DataInterface() {propertyValues {Curve {keys [{tangentMode=RCTM_User, time=0, value=0}, {tangentMode=RCTM_User, time=1, value=1}]}}})
    Module.Param0WriteEnabled=false
    Module.Param1WriteEnabled=true
    Module.Param2WriteEnabled=false
    Module.Param3WriteEnabled=true
   )
   &Color(Color=&MakeLinearColorFromVectorAndFloat(
     "Vector (RGB)"=<1 1 1>
     "Float (Alpha)"=&FloatFromCurve(FloatCurve=DataInterface() {propertyValues {Curve {keys [<at 0.5 1 RCIM_Cubic>,<at 1 0>]}}})
    ))
  ]
  rendererClasses [&NiagaraSpriteRendererProperties]
  renderers [{rendererClass=&NiagaraSpriteRendererProperties}]
 }
}
summary {
 systemName=NS_NDC_Jethro_WaterInteraction
 emitters [
  {emitterName=Impact_Surface_Foam, rendererClasses [&NiagaraMeshRendererProperties]}
  {emitterName=Drops_Wide001, rendererClasses [&NiagaraSpriteRendererProperties]}
  {emitterName=Drops_Tall, rendererClasses [&NiagaraSpriteRendererProperties]}
  {emitterName=Splash, rendererClasses [&NiagaraSpriteRendererProperties]}
  {emitterName=Splash001, rendererClasses [&NiagaraSpriteRendererProperties]}
  {emitterName=Base_Flash, rendererClasses [&NiagaraSpriteRendererProperties, &NiagaraMeshRendererProperties]}
 ]
 userVariables []
}
